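; Blueprint for a variant of the ornate side table furniture prop, with an ornate vase as a secondary mesh.
; Its construction script runs the parent's logic, then picks a random entry from the TintColor array and applies it to the vase.
; The random stream is seeded from the actor's location, so the tint is deterministic for each placement.
; BeginPlay, overlap and Tick events are empty stubs.
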(event EventBeginPlay)

(event Collision|EventActorBeginOverlap
  (params (OtherActor :type Actor%20Object%20Reference)))

(event EventTick
  (params (DeltaSeconds :type Float%20%28single-precision%29)))

(fn ConstructionScript
  (|CallParentFunction)
  (Default|SetColoronSecondaryMesh (Utilities|Array|RandomArrayItemfromStream (Variables|Default|GetTintColor) (Math|Random|MakeRandomStreamfromLocation (Transformation|GetActorLocation) 1.0 true)) (Variables|Default|GetSM_Container_OrnateVase_B)))
